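# UI-domain display material for the TextureDuplicateFinder editor utility, showing either the RGB or the alpha of a texture.
# A Texture parameter feeds a named reroute, and the ShowRGB static switch selects its RGB or alpha mask.
# A separate lerp blends TextureColor and TextureLinearColor parameters by UseLinearColor, with no link to the switch branch.
# The graph shows no connection from either branch to a material output.
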Material /Game/NeverCook/EditorUtilities/TextureDuplicateFinder/M_TDF_Display
graph {
 properties {bAutomaticallySetUsageInEditor=false, materialDomain=MD_UI, usedShadingModels=Default Lit}
 nodes [
  LinearInterpolate(
   A=TextureSampleParameter2D() {parameterName=TextureColor, texture=&T_BlacksRGB}.RGBA
   B=TextureSampleParameter2D() {parameterName=TextureLinearColor, texture=&T_BlacksRGB}.RGBA
   Alpha=ScalarParameter() {parameterName=UseLinearColor}
  )
  NamedRerouteDeclaration(None=TextureSampleParameter2D() {parameterName=Texture, texture=&T_BlacksRGB}.RGBA) {name=RGBA_Route, nodeColor=<rgba 0 1 0.9063 1>}
  n1 = NamedRerouteUsage()
  StaticSwitchParameter(True=ComponentMask(None=@n1) {b=true, g=true, r=true}, False=ComponentMask(None=@n1) {a=true}) {defaultValue=true, parameterName=ShowRGB}
 ]
}
