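# Editor utility list-row widget for the Texture Duplicate Finder tool, showing one matched texture with RGB and alpha previews, package name and match type.
# When its list item is set, it casts to ListData_TDF, stores the asset data and labels the match as color only or color plus alpha.
# SetAssetData loads the Texture2D into both preview images' dynamic materials.
# The find button syncs the Content Browser to that asset.
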EditorUtilityWidgetBlueprint /Game/NeverCook/EditorUtilities/TextureDuplicateFinder/EUW_TDF_ListItem
graphs [
 EventGraph {nodes [
   "On Clicked (btnFind)"() {is=title K2Node_ComponentBoundEvent}
   n1 = Break Asset Data(AssetData=Get CurrentAssetData() -> (CurrentAssetData:Asset Data Structure))
   > SyncBrowserToObjects(AssetPaths=Make Array("[0]":String="To String (Name)"(InName=@n1)) -> (Array:Array of Strings))
   n2 = Event OnListItemObjectSet()
   > n3 = Cast To ListData_TDF!(execute, Object=@n2.ListItemObject) -> (then, CastFailed, AsList Data TDF:List Data TDF Object Reference) {is=title K2Node_DynamicCast}
   > SetAssetData!(execute, self, CurrentAssetData:Asset Data Structure=Get AssetData!(self:List Data TDF Object Reference=@n3) -> (AssetData)) -> (then)
   n4 = Get ColorAndAlphaMatch!(self:List Data TDF Object Reference=@n3) -> (ColorAndAlphaMatch:bool)
   > "SetText (Text)"(
    self=Get txtMatch() -> (txtMatch:Text Object Reference)
    InText=Format Text!(
     Format="LOCTEXT(\"Color Match{alpha}\")"
     alpha:Text=Select(Option 0:Text, Option 1:Text="LOCTEXT(\"\\r\\nAlpha Match\")", Index=@n4) -> (ReturnValue:Text)
    ) -> (Result)
   )
  ]}
 SetAssetData {nodes [
   n1 = SetAssetData!() -> (then, CurrentAssetData:Asset Data Structure) {is=title K2Node_FunctionEntry}
   > n2 = Set CurrentAssetData(CurrentAssetData:Asset Data Structure=@n1) -> (Output_Get:Asset Data Structure)
   > n3 = Cast To Texture2D(Object=GetAsset(InAssetData=@n2))
   > Set CurrentTexture(CurrentTexture:Texture 2D Object Reference=@n3) -> (Output_Get:Texture 2D Object Reference)
   > n4 = GetDynamicMaterial(self=Get ImageRGB() -> (ImageRGB:Image Object Reference))
   > SetTextureParameterValue(self=@n4, ParameterName=Texture, Value=Get CurrentTexture() -> (CurrentTexture:Texture 2D Object Reference))
   > n5 = GetDynamicMaterial(self=Get ImageAlpha() -> (ImageAlpha:Image Object Reference))
   > SetTextureParameterValue(self=@n5, ParameterName=Texture, Value=Get CurrentTexture())
   n6 = Get CurrentAssetData()
   > "SetText (Text)"(
    self=Get txtName() -> (txtName:Text Object Reference)
    InText="To Text (Name)"!(InName:Name=Break Asset Data!(AssetData=@n6) -> (PackageName)) -> (ReturnValue:Text)
   )
  ]}
]
variables [
 CurrentAssetData {type=AssetData, default {}}
 CurrentTexture {type=Texture2D}
]
widgets {
 info {rootWidgetClass=&Overlay, widgetCount=11}
 widgets [
  Overlay_0 {widgetClassPath=&Overlay, widget=Overlay_0}
  Image_57 {
   widgetClassPath=&Image
   bIsVariable=true
   parent=Overlay_0
   widget=Image_57
   properties {brush {
     bIsDynamicallyLoaded=false
     drawAs=Image
     imageType=NoImage
     mirroring=NoMirror
     resourceName=None
     resourceObject=null
     tiling=NoTile
     imageSize=<3 3>
     margin=<ltrb 0 0 0 0>
     outlineSettings {
      bUseBrushTransparency=false
      roundingType=HalfHeightRadius
      width=0
      color {colorUseRule=UseColor_Specified, specifiedColor=<rgba 0 0 0 0>}
      cornerRadii=<0 0 0 0>
     }
     tintColor {colorUseRule=UseColor_Specified, specifiedColor=<rgba 0.0212 0.0212 0.0212 1>}
     uVRegion {bIsValid=false, max=<0 0>, min=<0 0>}
    }}
   slot {class=OverlaySlot, properties {horizontalAlignment=HAlign_Fill, verticalAlignment=VAlign_Fill}}
  }
  VerticalBox_2 {
   widgetClassPath=&VerticalBox
   parent=Overlay_0
   widget=VerticalBox_2
   slot {class=OverlaySlot, properties {horizontalAlignment=HAlign_Fill}}
  }
  StackBox_0 {widgetClassPath=&StackBox, parent=VerticalBox_2, widget=StackBox_0, slot {class=VerticalBoxSlot}}
  btnFind {
   widgetClassPath=&EditorUtilityButton
   parent=StackBox_0
   widget=btnFind
   properties {widgetStyle {
     clickedSlateSound {resourceObject=null}
     disabled {
      bIsDynamicallyLoaded=false
      drawAs=RoundedBox
      imageType=NoImage
      mirroring=NoMirror
      resourceName=None
      resourceObject=null
      tiling=NoTile
      imageSize=<0 0>
      margin=<ltrb 0 0 0 0>
      outlineSettings {
       bUseBrushTransparency=false
       roundingType=FixedRadius
       width=1
       color {colorUseRule=UseColor_ColorTable, specifiedColor=<rgba 0 0 0 0>}
       cornerRadii=<4 4 4 4>
      }
      tintColor {colorUseRule=UseColor_Specified, specifiedColor=<rgba 0.0395 0.0395 0.0395 1>}
      uVRegion {bIsValid=false, max=<0 0>, min=<0 0>}
     }
     disabledForeground {colorUseRule=UseColor_ColorTable, specifiedColor=<rgba 0 0 0 0>}
     hovered {
      bIsDynamicallyLoaded=false
      drawAs=RoundedBox
      imageType=NoImage
      mirroring=NoMirror
      resourceName=None
      resourceObject=null
      tiling=NoTile
      imageSize=<32 32>
      margin=<ltrb 0 0 0 0>
      outlineSettings {
       bUseBrushTransparency=false
       roundingType=FixedRadius
       width=1
       color {colorUseRule=UseColor_ColorTable, specifiedColor=<rgba 0 0 0 0>}
       cornerRadii=<4 4 4 4>
      }
      tintColor {colorUseRule=UseColor_Specified, specifiedColor=<rgba 0.0953 0.0953 0.0953 1>}
      uVRegion {bIsValid=false, max=<0 0>, min=<0 0>}
     }
     hoveredForeground {colorUseRule=UseColor_ColorTable, specifiedColor=<rgba 0 0 0 0>}
     hoveredSlateSound {resourceObject=null}
     normal {
      bIsDynamicallyLoaded=false
      drawAs=RoundedBox
      imageType=NoImage
      mirroring=NoMirror
      resourceName=None
      resourceObject=null
      tiling=NoTile
      imageSize=<32 32>
      margin=<ltrb 0 0 0 0>
      outlineSettings {
       bUseBrushTransparency=false
       roundingType=FixedRadius
       width=1
       color {colorUseRule=UseColor_ColorTable, specifiedColor=<rgba 0 0 0 0>}
       cornerRadii=<0 0 0 0>
      }
      tintColor {colorUseRule=UseColor_Specified, specifiedColor=<rgba 0.0395 0.0395 0.0395 1>}
      uVRegion {bIsValid=false, max=<0 0>, min=<0 0>}
     }
     normalForeground {colorUseRule=UseColor_ColorTable, specifiedColor=<rgba 0 0 0 0>}
     normalPadding=<ltrb 0 0 0 0>
     pressed {
      bIsDynamicallyLoaded=false
      drawAs=RoundedBox
      imageType=NoImage
      mirroring=NoMirror
      resourceName=None
      resourceObject=null
      tiling=NoTile
      imageSize=<32 32>
      margin=<ltrb 0 0 0 0>
      outlineSettings {
       bUseBrushTransparency=false
       roundingType=FixedRadius
       width=1
       color {colorUseRule=UseColor_ColorTable, specifiedColor=<rgba 0 0 0 0>}
       cornerRadii=<4 4 4 4>
      }
      tintColor {colorUseRule=UseColor_Specified, specifiedColor=<rgba 0.0284 0.0284 0.0284 1>}
      uVRegion {bIsValid=false, max=<0 0>, min=<0 0>}
     }
     pressedForeground {colorUseRule=UseColor_ColorTable, specifiedColor=<rgba 0 0 0 0>}
     pressedPadding=<ltrb 0 0 0 0>
     pressedSlateSound {resourceObject=null}
    }}
   slot {class=StackBoxSlot, properties {horizontalAlignment=HAlign_Left, padding=<ltrb 0 0 5 0>}}
  }
  Image_1 {
   widgetClassPath=&Image
   bIsVariable=true
   parent=btnFind
   widget=Image_1
   properties {brush {
     bIsDynamicallyLoaded=false
     drawAs=Image
     imageType=FullColor
     mirroring=NoMirror
     resourceName=None
     resourceObject=/Game/Athena/UI/Challenges/Art/ChallengeTypeIcons/UI_ChallengeIcon_Find
     tiling=NoTile
     imageSize=<16 16>
     margin=<ltrb 0 0 0 0>
     outlineSettings {
      bUseBrushTransparency=false
      roundingType=HalfHeightRadius
      width=0
      color {colorUseRule=UseColor_Specified, specifiedColor=<rgba 0 0 0 0>}
      cornerRadii=<0 0 0 0>
     }
     tintColor {colorUseRule=UseColor_Specified, specifiedColor=<rgba 1 1 1 1>}
     uVRegion {bIsValid=false, max=<0 0>, min=<0 0>}
    }}
   slot {class=ButtonSlot, properties {padding=<ltrb 1 1 1 1>}}
  }
  ImageRGB {
   widgetClassPath=&Image
   bIsVariable=true
   parent=StackBox_0
   widget=ImageRGB
   properties {brush {
     bIsDynamicallyLoaded=false
     drawAs=Image
     imageType=FullColor
     mirroring=NoMirror
     resourceName=None
     resourceObject=&/M_TDF_Display
     tiling=NoTile
     imageSize=<64 64>
     margin=<ltrb 0 0 0 0>
     outlineSettings {
      bUseBrushTransparency=false
      roundingType=HalfHeightRadius
      width=0
      color {colorUseRule=UseColor_Specified, specifiedColor=<rgba 0 0 0 0>}
      cornerRadii=<0 0 0 0>
     }
     tintColor {colorUseRule=UseColor_Specified, specifiedColor=<rgba 1 1 1 1>}
     uVRegion {bIsValid=false, max=<0 0>, min=<0 0>}
    }}
   slot {class=StackBoxSlot, properties {verticalAlignment=VAlign_Top}}
  }
  ImageAlpha {
   widgetClassPath=&Image
   bIsVariable=true
   parent=StackBox_0
   widget=ImageAlpha
   properties {brush {
     bIsDynamicallyLoaded=false
     drawAs=Image
     imageType=FullColor
     mirroring=NoMirror
     resourceName=None
     resourceObject=&/MI_TDF_Display_Alpha
     tiling=NoTile
     imageSize=<64 64>
     margin=<ltrb 0 0 0 0>
     outlineSettings {
      bUseBrushTransparency=false
      roundingType=HalfHeightRadius
      width=0
      color {colorUseRule=UseColor_Specified, specifiedColor=<rgba 0 0 0 0>}
      cornerRadii=<0 0 0 0>
     }
     tintColor {colorUseRule=UseColor_Specified, specifiedColor=<rgba 1 1 1 1>}
     uVRegion {bIsValid=false, max=<0 0>, min=<0 0>}
    }}
   slot {class=StackBoxSlot, properties {verticalAlignment=VAlign_Top, padding=<ltrb 5 0 0 0>}}
  }
  VerticalBox_0 {
   widgetClassPath=&VerticalBox
   parent=StackBox_0
   widget=VerticalBox_0
   slot {class=StackBoxSlot, properties {padding=<ltrb 5 0 5 0>, size {sizeRule=Fill, value=1}}}
  }
  txtName {
   widgetClassPath=&TextBlock
   bIsVariable=true
   parent=VerticalBox_0
   widget=txtName
   properties {
    text=Text Block
    font {
     bForceMonospaced=false
     bMaterialIsStencil=false
     fontMaterial=null
     fontObject=&Roboto
     letterSpacing=0
     monospacedWidth=1
     size=12
     skewAmount=0
     typefaceFontName=Regular
     outlineSettings {
      bApplyOutlineToDropShadows=false
      bMiteredCorners=false
      bSeparateFillAlpha=false
      outlineMaterial=null
      outlineSize=0
      outlineColor=<rgba 0 0 0 1>
     }
    }
   }
   slot {class=VerticalBoxSlot, properties {verticalAlignment=VAlign_Center}}
  }
  txtMatch {
   widgetClassPath=&TextBlock
   bIsVariable=true
   parent=VerticalBox_0
   widget=txtMatch
   properties {
    colorAndOpacity {colorUseRule=UseColor_Specified, specifiedColor=<rgba 0.5 0.5 0.5 1>}
    font {
     bForceMonospaced=false
     bMaterialIsStencil=false
     fontMaterial=null
     fontObject=&Roboto
     letterSpacing=0
     monospacedWidth=1
     size=12
     skewAmount=0
     typefaceFontName=Regular
     outlineSettings {
      bApplyOutlineToDropShadows=false
      bMiteredCorners=false
      bSeparateFillAlpha=false
      outlineMaterial=null
      outlineSize=0
      outlineColor=<rgba 0 0 0 1>
     }
    }
   }
   slot {class=VerticalBoxSlot, properties {verticalAlignment=VAlign_Center, size {sizeRule=Fill, value=1}}}
  }
 ]
}
