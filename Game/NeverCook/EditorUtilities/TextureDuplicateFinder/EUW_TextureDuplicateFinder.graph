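# Editor utility widget that finds likely duplicate Texture2D assets of the selected texture, or of a debug target while IsDebugging is on.
# It pre-filters asset registry textures by the Dimensions tag, then loads candidates and compares per-channel source min/max within Tolerance, a heuristic rather than a pixel comparison.
# Matches fill a list as color-and-alpha or color-only, and a details view exposes the widget's own properties.
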EditorUtilityWidgetBlueprint /Game/NeverCook/EditorUtilities/TextureDuplicateFinder/EUW_TextureDuplicateFinder
graphs [
 CompareDataToSource {nodes [
   CompareDataToSource!() -> (then, TestData:"Asset Data Structure (by ref)", SourceTexture:Texture 2D Object Reference) {is=title K2Node_FunctionEntry}
   > n1 = Sequence!(execute) -> (then_0, then_1, then_2:Exec)
   n2 = MakeLiteralName(Value=Dimensions)
   n3 = Get TestData() -> (TestData:"Asset Data Structure (by ref)")
   n4 = Get SourceTexture() -> (SourceTexture:Texture 2D Object Reference)
   n5 = "To String (Integer)"(InInt=GetSizeX!(self:Texture 2D Object Reference=@n4) -> (ReturnValue:int=0))
   n6 = "To String (Integer)"(InInt=GetSizeY!(self:Texture 2D Object Reference=@n4) -> (ReturnValue:int=0))
   > n7 = Branch(Condition="Equal, Case Insensitive (String)"(A=GetTagValue(InAssetData=@n3, InTagName=@n2), B=Append!(A=@n5, B=x, C:String=@n6) -> (ReturnValue)))
   Return Node!(execute=@n7.else, IsPossibleMatch:bool=false)
   Return Node(execute=@n1.then_2, IsPossibleMatch=true)
  ]}
 CompareTextureToSource {nodes [
   CompareTextureToSource!() -> (then, TestTexture:Texture 2D Object Reference, SourceTexture:Texture 2D Object Reference) {is=title K2Node_FunctionEntry}
   > n1 = Sequence!(execute) -> (then_0, then_1, then_2, then_3:Exec)
   > n2 = ComputeTextureSourceChannelMinMax!(execute, self:Texture Object Reference=Get TestTexture() -> (TestTexture:Texture 2D Object Reference)) -> (then, OutColorMin:Linear Color Structure, OutColorMax:Linear Color Structure, ReturnValue:bool)
   > Set l_testMin(l_testMin:Linear Color Structure=@n2) -> (Output_Get:Linear Color Structure)
   > Set l_testMax(l_testMax:Linear Color Structure=@n2.OutColorMax) -> (Output_Get:Linear Color Structure)
   n3 = BreakColor(InColor=Get SourceMin() -> (SourceMin:Linear Color Structure))
   n4 = MakeVector(X=@n3, Y=@n3.G, Z=@n3.B)
   n5 = BreakColor(InColor=Get l_testMin() -> (l_testMin:Linear Color Structure))
   n6 = MakeVector(X=@n5, Y=@n5.G, Z=@n5.B)
   n7 = Get Tolerance() -> (Tolerance:float)
   n8 = BreakColor(InColor=Get SourceMax() -> (SourceMax:Linear Color Structure))
   n9 = MakeVector(X=@n8, Y=@n8.G, Z=@n8.B)
   n10 = BreakColor(InColor=Get l_testMax() -> (l_testMax:Linear Color Structure))
   n11 = MakeVector(X=@n10, Y=@n10.G, Z=@n10.B)
   Set l_ColorMatch(
    execute=@n1.then_1
    l_ColorMatch:bool=AND Boolean(
     A="Equal (Vector)"!(A:Vector=@n4, B:Vector=@n6, ErrorTolerance:float=@n7) -> (ReturnValue:bool=~) {is=title K2Node_PromotableOperator}
     B="Equal (Vector)"(A=@n9, B=@n11, ErrorTolerance=@n7) -> (ReturnValue=~)
    )
   ) -> (Output_Get:bool)
   n12 = BreakColor(InColor=Get l_testMax())
   n13 = BreakColor(InColor=Get SourceMax())
   n14 = Get Tolerance()
   n15 = BreakColor(InColor=Get SourceMin())
   n16 = BreakColor(InColor=Get l_testMin())
   Set l_AlphaMatch(
    execute=@n1.then_2
    l_AlphaMatch:bool=AND Boolean(A="Nearly Equal (Float)"(A=@n15.A, B=@n16.A, ErrorTolerance=@n14), B="Nearly Equal (Float)"(A=@n13.A, B=@n12.A, ErrorTolerance=@n14))
   ) -> (Output_Get:bool)
   Return Node!(execute=@n1.then_3, IsColorMatch:bool=Get l_ColorMatch() -> (l_ColorMatch:bool), IsAlphaMatch:bool=Get l_AlphaMatch() -> (l_AlphaMatch:bool))
  ]}
 EventGraph {nodes [
   "On Clicked (TA_Generic01_Button)"!() -> (OutputDelegate, then, widget:User Widget Object Reference, event:Pointer Event Structure) {is=title K2Node_ComponentBoundEvent}
   > n1 = GetSourceTextureFromSelection!(execute, self) -> (then, Texture:Texture 2D Object Reference)
   > Is Valid(InputObject=@n1)
   > n2 = FindTextureDuplicates!(execute, self, SourceTexture:Texture 2D Object Reference=@n1) -> (then, ColorAndAlpha:Array of Texture 2D Object References, ColorOnly:Array of Texture 2D Object References)
   > UpdateList!(execute, self, ColorAndAlpha:Array of Texture 2D Object References=@n2, ColorOnly:Array of Texture 2D Object References=@n2.ColorOnly) -> (then)
   Event Construct()
   > SetObject(self=Get DetailsView_150() -> (DetailsView_150:Details View Object Reference), NewObject=Self-Reference())
  ]}
 FindTextureDuplicates {nodes [
   n1 = FindTextureDuplicates!() -> (then, SourceTexture:Texture 2D Object Reference) {is=title K2Node_FunctionEntry}
   > Is Valid(InputObject=@n1)
   > n2 = Sequence()
   n3 = MakeTopLevelAssetPath!(PackageName:String=/Script/Engine, AssetName:String=Texture2D) -> (ReturnValue:Top Level Asset Path Structure)
   > n4 = GetAssets!(
    execute
    self:Asset Registry Interface=GetAssetRegistry()
    Filter:"ARFilter Structure (by ref)"=MakeARFilter!(
     PackageNames:Array of Names
     PackagePaths:Array of Names=Get l_paths() -> (l_paths:Array of Names)
     SoftObjectPaths:Array of Soft Object Path Structures
     ClassPaths:Array of Top Level Asset Path Structures=Make Array("[0]":Top Level Asset Path Structure=@n3) -> (Array:Array of Top Level Asset Path Structures)
     RecursiveClassPathsExclusionSet:Set of Top Level Asset Path Structures
     ClassNames:Array of Names
     RecursiveClassesExclusionSet:Set of Names
     bRecursivePaths:bool=true
     bRecursiveClasses:bool=false
     bIncludeOnlyOnDiskAssets:bool=false
    ) -> (ReturnValue:ARFilter Structure)
    bSkipARFilteredAssets:bool=true
   ) -> (then, OutAssetData:Array of Asset Data Structures, ReturnValue:bool)
   > n5 = Branch(Condition=Is Valid Index!(TargetArray:Array of Asset Data Structures=@n4, IndexToTest:int) -> (ReturnValue:bool))
   > n6 = Set l_textureDatas(l_textureDatas:Array of Asset Data Structures=@n4) -> (Output_Get:Array of Asset Data Structures)
   > Remove Item!(
    execute
    TargetArray:Array of Asset Data Structures=@n6
    Item:"Asset Data Structure (by ref)"=CreateAssetData(InAsset=Get SourceTexture(), bAllowBlueprintClass=false)
   ) -> (then, ReturnValue:bool)
   Return Node!(execute=@n5.else, ColorAndAlpha:Array of Texture 2D Object References, ColorOnly:Array of Texture 2D Object References)
   n7 = For Each Loop(Exec=@n2.then_1, Array:Array of Asset Data Structures=Get l_textureDatas() -> (l_textureDatas:Array of Asset Data Structures)) -> (Array Element:Asset Data Structure)
   > n8 = CompareDataToSource!(execute, self, TestData:"Asset Data Structure (by ref)"=@n7, SourceTexture:Texture 2D Object Reference=Get SourceTexture()) -> (then, IsPossibleMatch:bool) {is=title K2Node_CallFunction}
   > Branch(Condition=@n8)
   > n9 = Cast To Texture2D(Object=GetAsset(InAssetData=@n7))
   > Add(
    TargetArray:Array of Texture 2D Object References=Get l_possibleMatches() -> (l_possibleMatches:Array of Texture 2D Object References)
    NewItem:"Texture 2D Object Reference (by ref)"=@n9
   ) {is=title K2Node_CallArrayFunction}
   n10 = For Each Loop(Exec=@n2.then_2, Array:Array of Texture 2D Object References=Get l_possibleMatches()) -> (Array Element:Texture 2D Object Reference)
   > n11 = CompareTextureToSource!(execute, self, TestTexture:Texture 2D Object Reference=@n10, SourceTexture:Texture 2D Object Reference=Get SourceTexture()) -> (then, IsColorMatch:bool, IsAlphaMatch:bool) {is=title K2Node_CallFunction}
   > Branch(Condition=@n11)
   > n12 = Branch(Condition=@n11.IsAlphaMatch)
   > Add(TargetArray=Get l_colorAndAlphaMatches() -> (l_colorAndAlphaMatches:Array of Texture 2D Object References), NewItem=@n10)
   Add(execute=@n12.else, TargetArray=Get l_colorOnlyMatches() -> (l_colorOnlyMatches:Array of Texture 2D Object References), NewItem=@n10)
   Return Node(execute=@n2.then_3, ColorAndAlpha=Get l_colorAndAlphaMatches(), ColorOnly=Get l_colorOnlyMatches())
  ]}
 GetSourceTextureFromSelection {nodes [
   GetSourceTextureFromSelection() {is=title K2Node_FunctionEntry}
   > n1 = Sequence!(execute) -> (then_0, then_1, then_2)
   > n2 = Branch(Condition=Get IsDebugging() -> (IsDebugging:bool))
   > Set l_textureSource(l_textureSource:Texture 2D Object Reference=Get TargetTextureDebug() -> (TargetTextureDebug:Texture 2D Object Reference)) -> (Output_Get:Texture 2D Object Reference)
   n3 = GetSelectedAssetData(execute=@n2.else)
   > n4 = For Each Loop with Break!(Exec:Exec, Array:Array of Asset Data Structures=@n3, Break:Exec=@n7.then) -> (LoopBody:Exec, Array Element:Asset Data Structure, Array Index:int, Completed:Exec)
   n5 = Break Asset Data!(AssetData=@n4) -> (AssetClassPath)
   > Branch(Condition="Equal, Case Insensitive (String)"(
     A=BreakTopLevelAssetPath!(TopLevelAssetPath:"Top Level Asset Path Structure (by ref)"=@n5) -> (PathString:String)
     B=/Script/Engine.Texture2D
    ))
   > n6 = Cast To Texture2D(Object=GetAsset(InAssetData=@n4))
   > n7 = Set l_textureSource(l_textureSource=@n6)
   n8 = ComputeTextureSourceChannelMinMax(execute=@n1.then_1, self=Get l_textureSource() -> (l_textureSource:Texture 2D Object Reference))
   > n9 = Branch(Condition=@n8.ReturnValue)
   > Set SourceMin(SourceMin:Linear Color Structure=@n8) -> (Output_Get:Linear Color Structure)
   > Set SourceMax(SourceMax:Linear Color Structure=@n8.OutColorMax) -> (Output_Get:Linear Color Structure)
   Return Node!(execute=@n9.else, Texture:Texture 2D Object Reference)
   SetAssetData!(
    execute=@n1.then_2
    self:EUW TDF List Item Object Reference=Get EUW_TDF_ListItem() -> (EUW_TDF_ListItem:EUW TDF List Item Object Reference)
    CurrentAssetData:Asset Data Structure=CreateAssetData(InAsset=Get l_textureSource(), bAllowBlueprintClass=false)
   ) -> (then)
   > Return Node(Texture=Get l_textureSource())
  ]}
 UpdateList {nodes [
   n1 = UpdateList!() -> (then, ColorAndAlpha:Array of Texture 2D Object References, ColorOnly:Array of Texture 2D Object References) {is=title K2Node_FunctionEntry}
   > ClearListItems(self=Get CommonListView_0() -> (CommonListView_0:Common List View Object Reference))
   > n2 = Sequence!(execute) -> (then_0, then_1)
   > n3 = For Each Loop(Array=@n1)
   > n4 = Construct Object from Class!(
    execute
    Class=&/ListData_TDF.ListData_TDF_C
    self
    AssetData:Asset Data Structure=CreateAssetData(InAsset=@n3, bAllowBlueprintClass=false)
    ColorAndAlphaMatch:bool=true
   ) -> (then, ReturnValue:List Data TDF Object Reference)
   > AddItem!(execute, self:List View Object Reference=Get CommonListView_0(), Item:Object Reference=@n4) -> (then)
   n5 = For Each Loop(Exec=@n2.then_1, Array=@n1.ColorOnly)
   > n6 = Construct Object from Class(
    Class=&ListData_TDF_C
    AssetData=CreateAssetData(InAsset=@n5, bAllowBlueprintClass=false)
    ColorAndAlphaMatch=False
   )
   > AddItem(self=Get CommonListView_0(), Item=@n6)
  ]}
]
variables [
 IsDebugging {type=boolean, default=true}
 SourceMax {type=LinearColor, default=<rgba 0 0 0 0>}
 SourceMin {type=LinearColor, default=<rgba 0 0 0 0>}
 TargetTextureDebug {type=Texture2D}
 Tolerance {type=number, default=0.0001}
]
widgets {
 info {rootWidgetClass=&HorizontalBox, widgetCount=7}
 widgets [
  HorizontalBox_0 {widgetClassPath=&HorizontalBox, widget=HorizontalBox_0}
  StackBox_0 {
   widgetClassPath=&StackBox
   parent=HorizontalBox_0
   widget=StackBox_0
   properties {orientation=Orient_Vertical}
   slot {class=HorizontalBoxSlot, properties {size {sizeRule=Fill, value=1}}}
  }
  TA_Generic01_Button {
   widgetClassPath=&TA_Generic01_Button
   parent=StackBox_0
   widget=TA_Generic01_Button
   properties {label=Find Duplicates Of Selected}
   slot {class=StackBoxSlot}
  }
  EUW_TDF_ListItem {
   widgetClassPath=&/EUW_TDF_ListItem
   bIsVariable=true
   parent=StackBox_0
   widget=EUW_TDF_ListItem
   slot {class=StackBoxSlot}
  }
  Image_226 {
   widgetClassPath=&Image
   bIsVariable=true
   parent=StackBox_0
   widget=Image_226
   properties {
    brush {
     bIsDynamicallyLoaded=false
     drawAs=Image
     imageType=NoImage
     mirroring=NoMirror
     resourceName=None
     resourceObject=null
     tiling=NoTile
     imageSize=<2 2>
     margin=<ltrb 0 0 0 0>
     outlineSettings {
      bUseBrushTransparency=false
      roundingType=HalfHeightRadius
      width=0
      color {colorUseRule=UseColor_Specified, specifiedColor=<rgba 0 0 0 0>}
      cornerRadii=<0 0 0 0>
     }
     tintColor {colorUseRule=UseColor_Specified, specifiedColor=<rgba 1 1 1 1>}
     uVRegion {bIsValid=false, max=<0 0>, min=<0 0>}
    }
    colorAndOpacity=<rgba 0 0 0 1>
   }
   slot {class=StackBoxSlot, properties {padding=<ltrb 0 5 0 5>}}
  }
  CommonListView_0 {
   widgetClassPath=&CommonListView
   bIsVariable=true
   parent=StackBox_0
   widget=CommonListView_0
   properties {
    entryWidgetClass=&/EUW_TDF_ListItem.EUW_TDF_ListItem_C
    verticalEntrySpacing=4
    scrollBarStyle {
     thickness=8
     draggedThumbImage {
      bIsDynamicallyLoaded=false
      drawAs=RoundedBox
      imageType=NoImage
      mirroring=NoMirror
      resourceName=None
      resourceObject=null
      tiling=NoTile
      imageSize=<0 0>
      margin=<ltrb 0 0 0 0>
      outlineSettings {
       bUseBrushTransparency=false
       roundingType=FixedRadius
       width=0
       color {colorUseRule=UseColor_Specified, specifiedColor=<rgba 0 0 0 0>}
       cornerRadii=<4 4 4 4>
      }
      tintColor {colorUseRule=UseColor_Specified, specifiedColor=<rgba 0.2159 0.2159 0.2159 1>}
      uVRegion {bIsValid=false, max=<0 0>, min=<0 0>}
     }
     horizontalBackgroundImage {
      bIsDynamicallyLoaded=false
      drawAs=NoDrawType
      imageType=NoImage
      mirroring=NoMirror
      resourceName=None
      resourceObject=null
      tiling=NoTile
      imageSize=<0 0>
      margin=<ltrb 0 0 0 0>
      outlineSettings {
       bUseBrushTransparency=false
       roundingType=HalfHeightRadius
       width=0
       color {colorUseRule=UseColor_Specified, specifiedColor=<rgba 0 0 0 0>}
       cornerRadii=<0 0 0 0>
      }
      tintColor {colorUseRule=UseColor_Specified, specifiedColor=<rgba 1 1 1 1>}
      uVRegion {bIsValid=false, max=<0 0>, min=<0 0>}
     }
     horizontalBottomSlotImage {
      bIsDynamicallyLoaded=false
      drawAs=NoDrawType
      imageType=NoImage
      mirroring=NoMirror
      resourceName=None
      resourceObject=null
      tiling=NoTile
      imageSize=<0 0>
      margin=<ltrb 0 0 0 0>
      outlineSettings {
       bUseBrushTransparency=false
       roundingType=HalfHeightRadius
       width=0
       color {colorUseRule=UseColor_Specified, specifiedColor=<rgba 0 0 0 0>}
       cornerRadii=<0 0 0 0>
      }
      tintColor {colorUseRule=UseColor_Specified, specifiedColor=<rgba 1 1 1 1>}
      uVRegion {bIsValid=false, max=<0 0>, min=<0 0>}
     }
     horizontalTopSlotImage {
      bIsDynamicallyLoaded=false
      drawAs=NoDrawType
      imageType=NoImage
      mirroring=NoMirror
      resourceName=None
      resourceObject=null
      tiling=NoTile
      imageSize=<0 0>
      margin=<ltrb 0 0 0 0>
      outlineSettings {
       bUseBrushTransparency=false
       roundingType=HalfHeightRadius
       width=0
       color {colorUseRule=UseColor_Specified, specifiedColor=<rgba 0 0 0 0>}
       cornerRadii=<0 0 0 0>
      }
      tintColor {colorUseRule=UseColor_Specified, specifiedColor=<rgba 1 1 1 1>}
      uVRegion {bIsValid=false, max=<0 0>, min=<0 0>}
     }
     hoveredThumbImage {
      bIsDynamicallyLoaded=false
      drawAs=RoundedBox
      imageType=NoImage
      mirroring=NoMirror
      resourceName=None
      resourceObject=null
      tiling=NoTile
      imageSize=<0 0>
      margin=<ltrb 0 0 0 0>
      outlineSettings {
       bUseBrushTransparency=false
       roundingType=FixedRadius
       width=0
       color {colorUseRule=UseColor_Specified, specifiedColor=<rgba 0 0 0 0>}
       cornerRadii=<4 4 4 4>
      }
      tintColor {colorUseRule=UseColor_Specified, specifiedColor=<rgba 0.2159 0.2159 0.2159 1>}
      uVRegion {bIsValid=false, max=<0 0>, min=<0 0>}
     }
     normalThumbImage {
      bIsDynamicallyLoaded=false
      drawAs=RoundedBox
      imageType=NoImage
      mirroring=NoMirror
      resourceName=None
      resourceObject=null
      tiling=NoTile
      imageSize=<0 0>
      margin=<ltrb 0 0 0 0>
      outlineSettings {
       bUseBrushTransparency=false
       roundingType=FixedRadius
       width=0
       color {colorUseRule=UseColor_Specified, specifiedColor=<rgba 0 0 0 0>}
       cornerRadii=<4 4 4 4>
      }
      tintColor {colorUseRule=UseColor_Specified, specifiedColor=<rgba 0.0953 0.0953 0.0953 1>}
      uVRegion {bIsValid=false, max=<0 0>, min=<0 0>}
     }
     verticalBackgroundImage {
      bIsDynamicallyLoaded=false
      drawAs=NoDrawType
      imageType=NoImage
      mirroring=NoMirror
      resourceName=None
      resourceObject=null
      tiling=NoTile
      imageSize=<0 0>
      margin=<ltrb 0 0 0 0>
      outlineSettings {
       bUseBrushTransparency=false
       roundingType=HalfHeightRadius
       width=0
       color {colorUseRule=UseColor_Specified, specifiedColor=<rgba 0 0 0 0>}
       cornerRadii=<0 0 0 0>
      }
      tintColor {colorUseRule=UseColor_Specified, specifiedColor=<rgba 1 1 1 1>}
      uVRegion {bIsValid=false, max=<0 0>, min=<0 0>}
     }
     verticalBottomSlotImage {
      bIsDynamicallyLoaded=false
      drawAs=NoDrawType
      imageType=NoImage
      mirroring=NoMirror
      resourceName=None
      resourceObject=null
      tiling=NoTile
      imageSize=<0 0>
      margin=<ltrb 0 0 0 0>
      outlineSettings {
       bUseBrushTransparency=false
       roundingType=HalfHeightRadius
       width=0
       color {colorUseRule=UseColor_Specified, specifiedColor=<rgba 0 0 0 0>}
       cornerRadii=<0 0 0 0>
      }
      tintColor {colorUseRule=UseColor_Specified, specifiedColor=<rgba 1 1 1 1>}
      uVRegion {bIsValid=false, max=<0 0>, min=<0 0>}
     }
     verticalTopSlotImage {
      bIsDynamicallyLoaded=false
      drawAs=NoDrawType
      imageType=NoImage
      mirroring=NoMirror
      resourceName=None
      resourceObject=null
      tiling=NoTile
      imageSize=<0 0>
      margin=<ltrb 0 0 0 0>
      outlineSettings {
       bUseBrushTransparency=false
       roundingType=HalfHeightRadius
       width=0
       color {colorUseRule=UseColor_Specified, specifiedColor=<rgba 0 0 0 0>}
       cornerRadii=<0 0 0 0>
      }
      tintColor {colorUseRule=UseColor_Specified, specifiedColor=<rgba 1 1 1 1>}
      uVRegion {bIsValid=false, max=<0 0>, min=<0 0>}
     }
    }
   }
   slot {class=StackBoxSlot, properties {size {sizeRule=Fill, value=1}}}
  }
  DetailsView_150 {
   widgetClassPath=&DetailsView
   parent=HorizontalBox_0
   widget=DetailsView_150
   properties {categoriesToShow [nocategory]}
   slot {class=HorizontalBoxSlot}
  }
 ]
}
